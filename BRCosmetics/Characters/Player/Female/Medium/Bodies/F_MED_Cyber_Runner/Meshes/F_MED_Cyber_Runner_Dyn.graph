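# Physics asset for secondary motion of the collar, jacket and forearm cloth on the female medium Cyber Runner body.
# Capsules on head, spine, clavicles and upper arms form the body; mostly thin boxes shape the collar and jacket flaps.
# Collar chains hang from spine_05, jacket chains from spine_03, forearm cloth from lower-arm twist bones.
# Small MSTR anchors have free angular constraints, while links below them are swing/twist limited, asymmetrically on the side collars.
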PhysicsAsset /BRCosmetics/Characters/Player/Female/Medium/Bodies/F_MED_Cyber_Runner/Meshes/F_MED_Cyber_Runner_Dyn
bodies {
 clavicle_l [{shapeType=Capsule, length=5.6323, radius=5.4481, center=<11.8115 -0.2768 -2.8247>, rotation=<rot 52.3814 0 0>}]
 clavicle_r [{shapeType=Capsule, length=4.4078, radius=5.7862, center=<-9.7166 0.984 1.1552>, rotation=<rot 55.5921 0 0>}]
 dyn_collar_bk_01 [{shapeType=Box, extentX=7.5, extentY=15.9871, extentZ=0.6903, center=<3.6782 0.0001 0>}]
 dyn_collar_bk_02 [{shapeType=Box, extentX=7.5, extentY=15.9871, extentZ=0.6903, center=<3.6782 0.0001 0>}]
 dyn_collar_bk_MSTR [{shapeType=Capsule, length=0.5, radius=0.5, rotation=<rot 90 0 0>}]
 dyn_collar_fr_l_01 [{shapeType=Box, extentX=7.5, extentY=1.6212, extentZ=0.6903, center=<3.7005 0.0001 -0.4673>, rotation=<rot 2.7293 0.0001 0>}]
 dyn_collar_fr_l_MSTR [{shapeType=Capsule, length=0.5, radius=0.5, rotation=<rot 90 0 0>}]
 dyn_collar_fr_r_01 [{shapeType=Box, extentX=7.5, extentY=1.4539, extentZ=0.6903, center=<-3.5289 0.0844 0.5874>, rotation=<rot -13.9746 -2.0443 8.4087>}]
 dyn_collar_fr_r_MSTR [{shapeType=Capsule, length=0.5, radius=0.5, rotation=<rot 90 0 0>}]
 dyn_collar_l_01 [
  {shapeType=Box, extentX=7.5, extentY=24.6275, extentZ=2.1903, center=<4.5845 3.0807 1.5249>, rotation=<rot -24.5987 -6.3912 11.0783>}
  {shapeType=Box, extentX=7.5, extentY=12.3198, extentZ=1.2485, center=<5.1217 8.167 -0.5222>, rotation=<rot -7 36.4301 8.7321>}
 ]
 dyn_collar_l_02 [{shapeType=Box, extentX=7.5, extentY=28.3735, extentZ=3.1903, center=<3.777 1.4201 0.1308>, rotation=<rot -16.1156 -0.5531 6.5972>}]
 dyn_collar_l_MSTR [{shapeType=Capsule, length=0.5, radius=0.5, rotation=<rot 90 0 0>}]
 dyn_collar_r_01 [
  {shapeType=Box, extentX=7.5, extentY=24.6275, extentZ=2.5, center=<-4.7606 -0.0977 -2.4238>, rotation=<rot -30.1148 -5.3983 -9.3214>}
  {shapeType=Box, extentX=7.5, extentY=12.3198, extentZ=1.2485, center=<-3.3375 -10.278 2.0004>, rotation=<rot -7 36.4289 26.7845>}
 ]
 dyn_collar_r_02 [{shapeType=Box, extentX=7.5, extentY=28.3735, extentZ=2.6903, center=<-3.8194 -1.1125 -0.2876>, rotation=<rot -16.1155 -0.5531 -7.273>}]
 dyn_collar_r_MSTR [{shapeType=Capsule, length=0.5, radius=0.5, rotation=<rot 90 0 0>}]
 dyn_forearm_cloth_l [{shapeType=Capsule, length=2, radius=2, rotation=<rot 90 0 0>}]
 dyn_forearm_cloth_r [{shapeType=Capsule, length=2, radius=2, center=<2 0 0>, rotation=<rot 90 0 0>}]
 dyn_jacket_bk_01 [{shapeType=Box, extentX=7.5, extentY=7.5, extentZ=0.6903, center=<3.6782 0.0001 0>}]
 dyn_jacket_bk_MSTR [{shapeType=Capsule, length=0.5, radius=0.5, rotation=<rot 90 0 0>}]
 dyn_jacket_bk_l_01 [{shapeType=Box, extentX=7.5, extentY=7.5, extentZ=0.6903, center=<3.6782 0.0001 0>}]
 dyn_jacket_bk_l_MSTR [{shapeType=Capsule, length=0.5, radius=0.5, rotation=<rot 90 0 0>}]
 dyn_jacket_bk_r_01 [{shapeType=Box, extentX=7.5, extentY=7.5, extentZ=0.6903, center=<-3.6782 0 0>}]
 dyn_jacket_bk_r_MSTR [{shapeType=Capsule, length=0.5, radius=0.5, rotation=<rot 90 0 0>}]
 dyn_jacket_fr_l_01 [{shapeType=Box, extentX=5.6362, extentY=7.5, extentZ=0.6903, center=<2.6448 0.0001 0>}]
 dyn_jacket_fr_l_02 [{shapeType=Box, extentX=4.3562, extentY=7.5, extentZ=0.6903, center=<2.0136 0.0001 0>}]
 dyn_jacket_fr_l_MSTR [{shapeType=Capsule, length=0.5, radius=0.5, rotation=<rot 90 0 0>}]
 dyn_jacket_fr_l_zipper [{shapeType=Box, extentX=3.0392, extentY=1.6811, extentZ=0.6903, center=<1.4606 0.0001 -0.0001>}]
 dyn_jacket_fr_r_01 [{shapeType=Box, extentX=5.6752, extentY=7.5, extentZ=0.6903, center=<-2.5092 0 0>}]
 dyn_jacket_fr_r_02 [{shapeType=Box, extentX=3.4334, extentY=7.5, extentZ=0.6903, center=<-1.7355 0 0>}]
 dyn_jacket_fr_r_MSTR [{shapeType=Capsule, length=0.5, radius=0.5, rotation=<rot 90 0 0>}]
 dyn_jacket_l_01 [{shapeType=Box, extentX=7.5, extentY=7.5, extentZ=0.6903, center=<3.6782 0.0001 0>}]
 dyn_jacket_l_MSTR [{shapeType=Capsule, length=0.5, radius=0.5, rotation=<rot 90 0 0>}]
 dyn_jacket_r_01 [{shapeType=Box, extentX=7.5, extentY=7.5, extentZ=0.6903, center=<-3.6782 0 0>}]
 dyn_jacket_r_MSTR [{shapeType=Capsule, length=0.5, radius=0.5, rotation=<rot 90 0 0>}]
 head [{shapeType=Capsule, length=2.1922, radius=6.2541, center=<1.933 -3.591 0>, rotation=<rot 90 0 0>}]
 lowerarm_twist_02_l [{shapeType=Capsule, length=0.5, radius=0.5, rotation=<rot 90 0 0>}]
 lowerarm_twist_02_r [{shapeType=Capsule, length=0.5, radius=0.5, rotation=<rot 90 0 0>}]
 spine_03 [{shapeType=Capsule, length=2.9926, radius=8.2172, center=<-2.9509 2.4864 0>}]
 spine_05 [{shapeType=Capsule, length=24.1045, radius=4.3947, center=<5.254 -0.0001 0>}]
 upperarm_l [{shapeType=Capsule, length=16.1562, radius=4.5126, center=<12.7155 -0.0006 0>, rotation=<rot 90 0 0>}]
 upperarm_r [{shapeType=Capsule, length=16.1562, radius=4.5126, center=<-12.7155 0 0>, rotation=<rot 90 0 0>}]
}
constraints [
 {
  bone1Name=lowerarm_twist_02_l
  bone2Name=upperarm_l
  swing1LimitDegrees=45
  swing1Motion=Free
  swing2LimitDegrees=45
  swing2Motion=Free
  twistLimitDegrees=45
  twistMotion=Free
 }
 {bone1Name=dyn_forearm_cloth_l, bone2Name=lowerarm_twist_02_l, swing1LimitDegrees=45, swing2LimitDegrees=45, twistLimitDegrees=5}
 {
  bone1Name=lowerarm_twist_02_r
  bone2Name=upperarm_r
  swing1LimitDegrees=45
  swing1Motion=Free
  swing2LimitDegrees=45
  swing2Motion=Free
  twistLimitDegrees=45
  twistMotion=Free
 }
 {bone1Name=dyn_forearm_cloth_r, bone2Name=lowerarm_twist_02_r, swing1LimitDegrees=45, swing2LimitDegrees=45, twistLimitDegrees=5}
 {
  bone1Name=dyn_collar_bk_MSTR
  bone2Name=spine_05
  swing1LimitDegrees=45
  swing1Motion=Free
  swing2LimitDegrees=45
  swing2Motion=Free
  twistLimitDegrees=45
  twistMotion=Free
 }
 {bone1Name=dyn_collar_bk_01, bone2Name=dyn_collar_bk_MSTR, swing1LimitDegrees=10, swing2LimitDegrees=45, twistLimitDegrees=5}
 {bone1Name=dyn_collar_bk_02, bone2Name=dyn_collar_bk_01, swing1LimitDegrees=10, swing2LimitDegrees=45, twistLimitDegrees=5}
 {
  bone1Name=dyn_collar_l_MSTR
  bone2Name=spine_05
  swing1LimitDegrees=45
  swing1Motion=Free
  swing2LimitDegrees=45
  swing2Motion=Free
  twistLimitDegrees=45
  twistMotion=Free
 }
 {bone1Name=dyn_collar_l_01, bone2Name=dyn_collar_l_MSTR, swing1LimitDegrees=5, swing2LimitDegrees=60, twistLimitDegrees=5}
 {bone1Name=dyn_collar_l_02, bone2Name=dyn_collar_l_01, swing1LimitDegrees=5, swing2LimitDegrees=60, twistLimitDegrees=5}
 {
  bone1Name=dyn_collar_fr_l_MSTR
  bone2Name=spine_05
  swing1LimitDegrees=45
  swing1Motion=Free
  swing2LimitDegrees=45
  swing2Motion=Free
  twistLimitDegrees=45
  twistMotion=Free
 }
 {bone1Name=dyn_collar_fr_l_01, bone2Name=dyn_collar_fr_l_MSTR, swing1LimitDegrees=45, swing2LimitDegrees=45, twistLimitDegrees=5}
 {
  bone1Name=dyn_collar_r_MSTR
  bone2Name=spine_05
  swing1LimitDegrees=45
  swing1Motion=Free
  swing2LimitDegrees=45
  swing2Motion=Free
  twistLimitDegrees=45
  twistMotion=Free
 }
 {bone1Name=dyn_collar_r_01, bone2Name=dyn_collar_r_MSTR, swing1LimitDegrees=20, swing2LimitDegrees=60, twistLimitDegrees=20}
 {bone1Name=dyn_collar_r_02, bone2Name=dyn_collar_r_01, swing1LimitDegrees=20, swing2LimitDegrees=60, twistLimitDegrees=20}
 {
  bone1Name=dyn_collar_fr_r_MSTR
  bone2Name=spine_05
  swing1LimitDegrees=45
  swing1Motion=Free
  swing2LimitDegrees=45
  swing2Motion=Free
  twistLimitDegrees=45
  twistMotion=Free
 }
 {bone1Name=dyn_collar_fr_r_01, bone2Name=dyn_collar_fr_r_MSTR, swing1LimitDegrees=45, swing2LimitDegrees=45, twistLimitDegrees=5}
 {
  bone1Name=dyn_jacket_fr_l_MSTR
  bone2Name=spine_03
  swing1LimitDegrees=45
  swing1Motion=Free
  swing2LimitDegrees=45
  swing2Motion=Free
  twistLimitDegrees=45
  twistMotion=Free
 }
 {bone1Name=dyn_jacket_fr_l_01, bone2Name=dyn_jacket_fr_l_MSTR, swing1LimitDegrees=45, swing2LimitDegrees=45, twistLimitDegrees=5}
 {bone1Name=dyn_jacket_fr_l_02, bone2Name=dyn_jacket_fr_l_01, swing1LimitDegrees=45, swing2LimitDegrees=45, twistLimitDegrees=5}
 {bone1Name=dyn_jacket_fr_l_zipper, bone2Name=dyn_jacket_fr_l_02, swing1LimitDegrees=45, swing2LimitDegrees=45, twistLimitDegrees=5}
 {
  bone1Name=dyn_jacket_l_MSTR
  bone2Name=spine_03
  swing1LimitDegrees=45
  swing1Motion=Free
  swing2LimitDegrees=45
  swing2Motion=Free
  twistLimitDegrees=45
  twistMotion=Free
 }
 {bone1Name=dyn_jacket_l_01, bone2Name=dyn_jacket_l_MSTR, swing1LimitDegrees=45, swing2LimitDegrees=45, twistLimitDegrees=5}
 {
  bone1Name=dyn_jacket_bk_l_MSTR
  bone2Name=spine_03
  swing1LimitDegrees=45
  swing1Motion=Free
  swing2LimitDegrees=45
  swing2Motion=Free
  twistLimitDegrees=45
  twistMotion=Free
 }
 {bone1Name=dyn_jacket_bk_l_01, bone2Name=dyn_jacket_bk_l_MSTR, swing1LimitDegrees=45, swing2LimitDegrees=45, twistLimitDegrees=5}
 {
  bone1Name=dyn_jacket_bk_MSTR
  bone2Name=spine_03
  swing1LimitDegrees=45
  swing1Motion=Free
  swing2LimitDegrees=45
  swing2Motion=Free
  twistLimitDegrees=45
  twistMotion=Free
 }
 {bone1Name=dyn_jacket_bk_01, bone2Name=dyn_jacket_bk_MSTR, swing1LimitDegrees=45, swing2LimitDegrees=45, twistLimitDegrees=5}
 {
  bone1Name=dyn_jacket_bk_r_MSTR
  bone2Name=spine_03
  swing1LimitDegrees=45
  swing1Motion=Free
  swing2LimitDegrees=45
  swing2Motion=Free
  twistLimitDegrees=45
  twistMotion=Free
 }
 {bone1Name=dyn_jacket_bk_r_01, bone2Name=dyn_jacket_bk_r_MSTR, swing1LimitDegrees=45, swing2LimitDegrees=45, twistLimitDegrees=5}
 {
  bone1Name=dyn_jacket_r_MSTR
  bone2Name=spine_03
  swing1LimitDegrees=45
  swing1Motion=Free
  swing2LimitDegrees=45
  swing2Motion=Free
  twistLimitDegrees=45
  twistMotion=Free
 }
 {bone1Name=dyn_jacket_r_01, bone2Name=dyn_jacket_r_MSTR, swing1LimitDegrees=45, swing2LimitDegrees=45, twistLimitDegrees=5}
 {
  bone1Name=dyn_jacket_fr_r_MSTR
  bone2Name=spine_03
  swing1LimitDegrees=45
  swing1Motion=Free
  swing2LimitDegrees=45
  swing2Motion=Free
  twistLimitDegrees=45
  twistMotion=Free
 }
 {bone1Name=dyn_jacket_fr_r_01, bone2Name=dyn_jacket_fr_r_MSTR, swing1LimitDegrees=45, swing2LimitDegrees=45, twistLimitDegrees=5}
 {bone1Name=dyn_jacket_fr_r_02, bone2Name=dyn_jacket_fr_r_01, swing1LimitDegrees=45, swing2LimitDegrees=45, twistLimitDegrees=5}
 {
  bone1Name=clavicle_l
  bone2Name=spine_05
  swing1LimitDegrees=45
  swing1Motion=Free
  swing2LimitDegrees=45
  swing2Motion=Free
  twistLimitDegrees=45
  twistMotion=Free
 }
 {
  bone1Name=clavicle_r
  bone2Name=spine_05
  swing1LimitDegrees=45
  swing1Motion=Free
  swing2LimitDegrees=45
  swing2Motion=Free
  twistLimitDegrees=45
  twistMotion=Free
 }
 {
  bone1Name=head
  bone2Name=spine_05
  swing1LimitDegrees=45
  swing1Motion=Free
  swing2LimitDegrees=45
  swing2Motion=Free
  twistLimitDegrees=45
  twistMotion=Free
 }
 {
  bone1Name=spine_05
  bone2Name=spine_03
  swing1LimitDegrees=45
  swing1Motion=Free
  swing2LimitDegrees=45
  swing2Motion=Free
  twistLimitDegrees=45
  twistMotion=Free
 }
 {
  bone1Name=upperarm_l
  bone2Name=clavicle_l
  swing1LimitDegrees=45
  swing1Motion=Free
  swing2LimitDegrees=45
  swing2Motion=Free
  twistLimitDegrees=45
  twistMotion=Free
 }
 {
  bone1Name=upperarm_r
  bone2Name=clavicle_r
  swing1LimitDegrees=45
  swing1Motion=Free
  swing2LimitDegrees=45
  swing2Motion=Free
  twistLimitDegrees=45
  twistMotion=Free
 }
]
