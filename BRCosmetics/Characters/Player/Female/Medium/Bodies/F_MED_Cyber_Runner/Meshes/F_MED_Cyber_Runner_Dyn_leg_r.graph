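# Physics asset for dynamic attachments on the right leg of the female medium Cyber Runner body: a pouch and a loincloth strip.
# Capsules cover the pelvis and upper right thigh, which are joined by an angularly free constraint.
# The pouch capsule hangs off the thigh twist bone with both swing axes locked.
# Four identical thin boxes form the cloth as a serial chain from that thigh bone, each joint having a much tighter twist limit than swing.
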PhysicsAsset /BRCosmetics/Characters/Player/Female/Medium/Bodies/F_MED_Cyber_Runner/Meshes/F_MED_Cyber_Runner_Dyn_leg_r
bodies {
 dyn_pouch [{shapeType=Capsule, length=2, radius=2, center=<0 0.6763 0>, rotation=<rot 90 0 0>}]
 dyn_roincloth_r_01 [{shapeType=Box, extentX=7.5, extentY=7.5, extentZ=0.6903, center=<3.6782 -3 0>}]
 dyn_roincloth_r_02 [{shapeType=Box, extentX=7.5, extentY=7.5, extentZ=0.6903, center=<3.6782 -3 0>}]
 dyn_roincloth_r_03 [{shapeType=Box, extentX=7.5, extentY=7.5, extentZ=0.6903, center=<3.6782 -3 0>}]
 dyn_roincloth_r_04 [{shapeType=Box, extentX=7.5, extentY=7.5, extentZ=0.6903, center=<3.6782 -3 0>}]
 pelvis [{shapeType=Capsule, length=10.4988, radius=10.4251, center=<-4.3848 -1.9103 -0.0001>, rotation=<rot -0.0004 0 0>}]
 thigh_twist_01_r [{shapeType=Capsule, length=30.6533, radius=7.5557, center=<-7.0172 0.1986 0.0001>, rotation=<rot 90 0 0>}]
}
constraints [
 {
  bone1Name=thigh_twist_01_r
  bone2Name=pelvis
  swing1LimitDegrees=45
  swing1Motion=Free
  swing2LimitDegrees=45
  swing2Motion=Free
  twistLimitDegrees=45
  twistMotion=Free
 }
 {
  bone1Name=dyn_pouch
  bone2Name=thigh_twist_01_r
  swing1LimitDegrees=45
  swing1Motion=Locked
  swing2LimitDegrees=45
  swing2Motion=Locked
  twistLimitDegrees=45
 }
 {bone1Name=dyn_roincloth_r_01, bone2Name=thigh_twist_01_r, swing1LimitDegrees=45, swing2LimitDegrees=45, twistLimitDegrees=5}
 {bone1Name=dyn_roincloth_r_02, bone2Name=dyn_roincloth_r_01, swing1LimitDegrees=45, swing2LimitDegrees=45, twistLimitDegrees=5}
 {bone1Name=dyn_roincloth_r_03, bone2Name=dyn_roincloth_r_02, swing1LimitDegrees=45, swing2LimitDegrees=45, twistLimitDegrees=5}
 {bone1Name=dyn_roincloth_r_04, bone2Name=dyn_roincloth_r_03, swing1LimitDegrees=45, swing2LimitDegrees=45, twistLimitDegrees=5}
]
