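# Physics asset for the Cyber Runner female cosmetic's face accessory part, defining bodies for dynamic hair strands, a ponytail and left and right strings.
# Small capsules form short chains with limited swing and twist, hair hanging from the head and the ponytail and strings from MSTR root bodies.
# The head and roots connect to spine_05 with free angular motion.
# Larger capsules and angled boxes on spine_04 and spine_05 form upper torso volumes.
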PhysicsAsset /BRCosmetics/Characters/Player/Female/Medium/Bodies/F_MED_Cyber_Runner/Meshes/Parts/F_MED_Cyber_Runner_FaceAcc_Dyn
bodies {
 dyn_hair_mid_01 [{shapeType=Capsule, length=2, radius=1, center=<1.2 0 0>, rotation=<rot 90 0 0>}]
 dyn_hair_mid_02 [{shapeType=Capsule, length=2, radius=1, center=<1.2 0 0>, rotation=<rot 90 0 0>}]
 dyn_hair_mid_l_01 [{shapeType=Capsule, length=2, radius=1, center=<1.2 0 0>, rotation=<rot 90 0 0>}]
 dyn_hair_mid_r_01 [{shapeType=Capsule, length=2, radius=1, center=<1.2 0 0>, rotation=<rot 90 0 0>}]
 dyn_hair_r_01 [{shapeType=Capsule, length=2, radius=1, center=<1.2 0 0>, rotation=<rot 90 0 0>}]
 dyn_hair_r_02 [{shapeType=Capsule, length=2, radius=1, center=<1.2 0 0>, rotation=<rot 90 0 0>}]
 dyn_hair_sml_l_01 [{shapeType=Capsule, length=2, radius=1, center=<0.1 0 0>, rotation=<rot 90 0 0>}]
 dyn_ponytail_01 [{shapeType=Capsule, length=2, radius=1, center=<1.3 0 0>, rotation=<rot 90 0 0>}]
 dyn_ponytail_02 [{shapeType=Capsule, length=2, radius=1, center=<1.3 0 0>, rotation=<rot 90 0 0>}]
 dyn_ponytail_MSTR [{shapeType=Capsule, length=0.505, radius=0.505, rotation=<rot 0 -90 0>}]
 dyn_string_l_01 [{shapeType=Capsule, length=2.5, radius=0.3, center=<1.5 0 0>, rotation=<rot 90 0 0>}]
 dyn_string_l_02 [{shapeType=Capsule, length=2.5, radius=0.3, center=<1.5 0 0>, rotation=<rot 90 0 0>}]
 dyn_string_l_03 [{shapeType=Capsule, length=2.5, radius=0.3, center=<1.5 0 0>, rotation=<rot 90 0 0>}]
 dyn_string_l_MSTR [{shapeType=Capsule, length=0.05, radius=0.05}]
 dyn_string_r_01 [{shapeType=Capsule, length=2.5, radius=0.3, center=<1.5 0 0>, rotation=<rot 90 0 0>}]
 dyn_string_r_02 [{shapeType=Capsule, length=2.5, radius=0.3, center=<1.5 0 0>, rotation=<rot 90 0 0>}]
 dyn_string_r_03 [{shapeType=Capsule, length=2.5, radius=0.3, center=<1.5 0 0>, rotation=<rot 90 0 0>}]
 dyn_string_r_MSTR [{shapeType=Capsule, length=0.05, radius=0.05}]
 head [{shapeType=Capsule, length=0.05, radius=0.05}]
 spine_04 [
  {shapeType=Capsule, length=5, radius=5, center=<-1.756 2.9681 0>}
  {shapeType=Box, extentX=16.6027, extentY=10.5, extentZ=8.1795, center=<-3.5093 12.74 -9.1116>, rotation=<rot -13.302 -15.6336 18.482>}
  {shapeType=Box, extentX=16.6027, extentY=7.5, extentZ=5.6795, center=<-2.8914 12.9595 7.2084>, rotation=<rot 7.2927 -11.7572 -19.4937>}
 ]
 spine_05 [
  {shapeType=Capsule, length=12.5685, radius=12.0017, center=<-7.5207 0.8264 0>}
  {shapeType=Box, extentX=16.6027, extentY=7.5, extentZ=8.1795, center=<-4.0571 12.2834 -10.8224>, rotation=<rot -13.302 -15.6333 18.4819>}
  {shapeType=Box, extentX=16.6027, extentY=7.5, extentZ=6.6795, center=<-3.684 12.5062 8.8453>, rotation=<rot 13.3583 -13.9608 -19.8896>}
 ]
}
constraints [
 {bone1Name=dyn_hair_mid_02, bone2Name=dyn_hair_mid_01, swing1LimitDegrees=45, swing2LimitDegrees=45, twistLimitDegrees=5}
 {bone1Name=dyn_hair_r_02, bone2Name=dyn_hair_r_01, swing1LimitDegrees=30, swing2LimitDegrees=2, twistLimitDegrees=2}
 {bone1Name=dyn_string_r_01, bone2Name=dyn_string_r_MSTR, swing1LimitDegrees=20, swing2LimitDegrees=45, twistLimitDegrees=3}
 {bone1Name=dyn_string_r_02, bone2Name=dyn_string_r_01, swing1LimitDegrees=20, swing2LimitDegrees=45, twistLimitDegrees=3}
 {bone1Name=dyn_string_r_03, bone2Name=dyn_string_r_02, swing1LimitDegrees=20, swing2LimitDegrees=45, twistLimitDegrees=3}
 {bone1Name=dyn_string_l_01, bone2Name=dyn_string_l_MSTR, swing1LimitDegrees=20, swing2LimitDegrees=45, twistLimitDegrees=3}
 {bone1Name=dyn_string_l_02, bone2Name=dyn_string_l_01, swing1LimitDegrees=20, swing2LimitDegrees=45, twistLimitDegrees=3}
 {bone1Name=dyn_string_l_03, bone2Name=dyn_string_l_02, swing1LimitDegrees=20, swing2LimitDegrees=45, twistLimitDegrees=3}
 {bone1Name=dyn_ponytail_01, bone2Name=dyn_ponytail_MSTR, swing1LimitDegrees=20, swing2LimitDegrees=20, twistLimitDegrees=5}
 {bone1Name=dyn_ponytail_02, bone2Name=dyn_ponytail_01, swing1LimitDegrees=20, swing2LimitDegrees=20, twistLimitDegrees=5}
 {bone1Name=dyn_hair_sml_l_01, bone2Name=head, swing1LimitDegrees=45, swing2LimitDegrees=45, twistLimitDegrees=5}
 {bone1Name=dyn_hair_mid_l_01, bone2Name=head, swing1LimitDegrees=45, swing2LimitDegrees=45, twistLimitDegrees=5}
 {bone1Name=dyn_hair_mid_01, bone2Name=head, swing1LimitDegrees=45, swing2LimitDegrees=45, twistLimitDegrees=5}
 {bone1Name=dyn_hair_mid_r_01, bone2Name=head, swing1LimitDegrees=45, swing2LimitDegrees=45, twistLimitDegrees=5}
 {bone1Name=dyn_hair_r_01, bone2Name=head, swing1LimitDegrees=30, swing2LimitDegrees=2, twistLimitDegrees=2}
 {
  bone1Name=head
  bone2Name=spine_05
  swing1LimitDegrees=45
  swing1Motion=Free
  swing2LimitDegrees=45
  swing2Motion=Free
  twistLimitDegrees=45
  twistMotion=Free
 }
 {
  bone1Name=dyn_string_r_MSTR
  bone2Name=spine_05
  swing1LimitDegrees=45
  swing1Motion=Free
  swing2LimitDegrees=45
  swing2Motion=Free
  twistLimitDegrees=45
  twistMotion=Free
 }
 {
  bone1Name=dyn_string_l_MSTR
  bone2Name=spine_05
  swing1LimitDegrees=45
  swing1Motion=Free
  swing2LimitDegrees=45
  swing2Motion=Free
  twistLimitDegrees=45
  twistMotion=Free
 }
 {
  bone1Name=dyn_ponytail_MSTR
  bone2Name=spine_05
  swing1LimitDegrees=45
  swing1Motion=Free
  swing2LimitDegrees=45
  swing2Motion=Free
  twistLimitDegrees=45
  twistMotion=Free
 }
 {
  bone1Name=spine_05
  bone2Name=spine_04
  swing1LimitDegrees=45
  swing1Motion=Free
  swing2LimitDegrees=45
  swing2Motion=Free
  twistLimitDegrees=45
  twistMotion=Free
 }
]
