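# Animation Blueprint for the Cyber Runner female face accessory part, adding secondary motion to its ponytail and two strings.
# It copies the attached parent mesh's pose, then weighted constraints make the dynamic master bones follow neck, head, upper spine and clavicles.
# A RigidBody node simulates the dedicated physics asset relative to spine_02, with overridden gravity and damped component motion.
# When bIsInFrontEnd is true, a post-simulation additive rotation re-angles the ponytail.
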AnimBlueprint /BRCosmetics/Characters/Player/Female/Medium/Bodies/F_MED_Cyber_Runner/Meshes/Parts/F_MED_Cyber_Runner_FaceAcc_AnimBP
target_skeleton=&/F_MED_Cyber_Runner_FaceAcc_Skeleton
graphs [
 AnimGraph {nodes [
   n1 = "Constraint - dyn_ponytail_MSTR (PPPP)"!(
    ComponentPose=Local To Component(LocalPose=Copy Pose From Mesh() {bUseAttachedParent=true})
    Alpha=1
    ConstraintWeights_0=0.1837
    ConstraintWeights_1:float=0.2692
    ConstraintWeights_2:float=0.1699
    ConstraintWeights_3:float=0.3772
   ) -> (Pose) {
    alphaBoolBlend {blendOption=Linear}
    boneToModify {boneName=dyn_ponytail_MSTR}
    constraintSetup [{targetBone {boneName=neck_02}}, {targetBone {boneName=head}}, {targetBone {boneName=neck_01}}, {targetBone {boneName=spine_05}}]
    constraintWeights [1,1,1,1]
   }
   n2 = "Constraint - dyn_string_l_MSTR (PP)"!(ComponentPose=@n1, Alpha=1, ConstraintWeights_0=0.9989, ConstraintWeights_1:float=0.0011) -> (Pose) {
    alphaBoolBlend {blendOption=Linear}
    boneToModify {boneName=dyn_string_l_MSTR}
    constraintSetup [{targetBone {boneName=spine_05}}, {targetBone {boneName=clavicle_l}}]
    constraintWeights [1,1]
   }
   n3 = "Constraint - dyn_string_r_MSTR (PP)"!(ComponentPose=@n2, Alpha=1, ConstraintWeights_0=0.0023, ConstraintWeights_1:float=0.9977) -> (Pose) {
    alphaBoolBlend {blendOption=Linear}
    boneToModify {boneName=dyn_string_r_MSTR}
    constraintSetup [{targetBone {boneName=clavicle_r}}, {targetBone {boneName=spine_05}}]
    constraintWeights [1,1]
   }
   n4 = RigidBody(ComponentPose=@n3) {
    bClampLinearTranslationLimitToRefPose=true
    bOverrideWorldGravity=true
    overridePhysicsAsset=&/F_MED_Cyber_Runner_FaceAcc_Dyn
    simulationSpace=BaseBoneSpace
    alphaBoolBlend {blendOption=Linear}
    baseBoneRef {boneName=spine_02}
    componentAppliedLinearAccClamp=<3000 3000 3000>
    componentLinearAccScale=<0.3 0.3 0.3>
    componentLinearVelScale=<0.3 0.3 0.3>
    overrideWorldGravity {z=-980}
   }
   n5 = "Transform (Modify) Bone - Bone: dyn_ponytail_MSTR"!(ComponentPose=@n4, bAlphaBoolEnabled:bool=Get bIsInFrontEnd(), Translation="0,0,0", Rotation="24,-8,0", Scale="1,1,1") -> (Pose) {
    alphaInputType=Bool
    rotationMode=BMM_Additive
    rotationSpace=BCS_BoneSpace
    alphaBoolBlend {blendOption=Linear}
    boneToModify {boneName=dyn_ponytail_MSTR}
    rotation {pitch=24, yaw=-8}
   }
   Output Pose(Result=Component To Local(ComponentPose=@n5)) {layerGroup=None}
  ]}
 EventGraph {nodes [
   TryGetPawnOwner()
   Event BlueprintUpdateAnimation()
  ]}
]
variables []
